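# Translucent, unlit, two-sided Niagara material that draws a soft, shimmering sun for the StrideMiceGiant emote.
# A vertical two-color gradient, tinted by particle color, drives emissive.
# An exponential radial gradient forms the disc, with UVs distorted by panning noise and shifted vertically by SunProgress so the sun can rise into view.
# Opacity combines the disc with a fade toward the bottom, particle alpha, view-angle, depth and capped camera-distance fades.
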Material /SparksCosmetics/Emotes/Group/StrideMiceDeep/FX/M_Emote_StrideMiceGiant_Sun
graph {
 properties {
  bAutomaticallySetUsageInEditor=false
  bUsedWithNiagaraMeshParticles=true
  bUsedWithNiagaraSprites=true
  blendMode=BLEND_Translucent
  shadingModel=MSM_Unlit
  twoSided=true
  usedShadingModels=Unlit
 }
 nodes [
  n1 = VectorParameter() {parameterName=SunColor1, defaultValue=<rgba 1 0.4519 0.1084 1>}
  n2 = VectorParameter() {parameterName=SunColor2, defaultValue=<rgba 1 0.0874 0.0492 1>}
  n3 = Add(A=TextureCoordinate(), B=ScalarParameter() {defaultValue=0.1, parameterName=ColorOffset}) {constB=0.5}
  n4 = ScalarParameter() {defaultValue=0.7, parameterName=ColorGradientSize}
  n5 = Saturate(None=ComponentMask(None=Divide(A=@n3, B=@n4) {constB=1}) {g=true})
  n6 = ParticleColor()
  MaterialFunctionCall(Emissive=Multiply(A=LinearInterpolate(A=@n1.RGB, B=@n2.RGB, Alpha=@n5), B=@n6.RGB)) {materialFunction=&MF_FX_EmissiveAdjustments}
  n7 = TextureSample(UVs=Panner(Coordinate=TextureCoordinate() {uTiling=0.01, vTiling=3}) {speedX=-0.1, speedY=0.07}) {samplerType=SAMPLERTYPE_LinearColor, texture=&HighContrastNoise}
  n8 = ScalarParameter() {defaultValue=0.03, parameterName=DistortionIntensity}
  n9 = ScalarParameter() {defaultValue=1, parameterName=SunProgress}
  n10 = Add(
   A=Add(A=TextureCoordinate(), B=Multiply(A=@n7.R, B=@n8) {constB=0.04})
   B=AppendVector(A=Constant(), B=LinearInterpolate(Alpha=@n9) {constA=-1, constB=0})
  )
  n11 = ScalarParameter() {defaultValue=0.4, parameterName=SunRadius}
  n12 = ScalarParameter() {defaultValue=20, parameterName=SunDensity}
  n13 = ComponentMask(None=TextureCoordinate()) {g=true}
  n14 = Add(A=Divide(A=OneMinus(None=@n13)) {constB=1}) {constB=0.2}
  n15 = Multiply(
   A=Multiply(
    A=MaterialFunctionCall(UVs=@n10, Radius=@n11, Density=@n12) {materialFunction=&RadialGradientExponential}.RadialGradientExponential
    B=Saturate(None=@n14)
   )
   B=@n6.A
  )
  n16 = MaterialFunctionCall("FadeSensitivity (optional)"=ScalarParameter() {defaultValue=2, parameterName=AngleFade}) {materialFunction=&FadeBasedOnViewAngle}
  n17 = MaterialFunctionCall!(Distance=ScalarParameter() {defaultValue=380, parameterName=CameraFade}) {materialFunction=/Game/Effects/Fort_Effects/Materials/_Functions/MF_CameraDistanceFade}
  Multiply(
   A=DepthFade(Opacity=Multiply(A=@n15, B=@n16.Result), FadeDistance=ScalarParameter() {defaultValue=50, parameterName=DepthFade})
   B=LinearInterpolate(B=ScalarParameter() {defaultValue=0.3, parameterName=DistanceFadeMaxOpacity}, Alpha=Saturate(None=@n17.Result))
  )
  SceneTexture()
 ]
}
