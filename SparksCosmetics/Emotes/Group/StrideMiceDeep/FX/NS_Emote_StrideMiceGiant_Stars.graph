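# Niagara effect for the StrideMiceGiant emote, wrapping a character in star flares, glows and a sun.
# Flares and a mobile aura follow skeletal mesh bones using per-partition shape and size arrays.
# InstrumentFlares kills spawn groups not matching User.InstrumentSelect, then fills per-instrument cylinder volumes.
# Audio reactivity modulates flare flicker, while sun progress, SunEnabled and front-end state drive the body glow mesh and sun sprites with lights.
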NiagaraSystem /SparksCosmetics/Emotes/Group/StrideMiceDeep/FX/NS_Emote_StrideMiceGiant_Stars
dependencies {
 usedDataInterfaces [
  &NiagaraDataInterfaceCurve
  &NiagaraDataInterfaceArrayFloat
  &NiagaraDataInterfaceArrayFloat3
  &NiagaraDataInterfaceArrayInt32
  &NiagaraDataInterfaceArrayFloat2
  &NiagaraDataInterfaceSkeletalMesh
 ]
 usedDynamicInputs [
  &InvertBool
  &Vector2DFromFloat
  &DistanceBetweenPositions
 ]
 usedModules [
  &self:SystemSpawnScript.SystemScriptSource.SystemScriptGraph.NiagaraNodeAssignment_5.SetVariables_#1
  &SampleSkeletalMesh
  &SystemState
  &BlendSpaceHelper
  &self:SystemSpawnScript.SystemScriptSource.SystemScriptGraph.NiagaraNodeAssignment_1.SetVariables_#2
  &self:SystemSpawnScript.SystemScriptSource.SystemScriptGraph.NiagaraNodeAssignment_0.SetVariables_#3
  &self:SystemSpawnScript.SystemScriptSource.SystemScriptGraph.NiagaraNodeAssignment_3.SetVariables_#4
  &self:SystemSpawnScript.SystemScriptSource.SystemScriptGraph.NiagaraNodeAssignment_2.SetVariables_#5
  &self:SystemSpawnScript.SystemScriptSource.SystemScriptGraph.NiagaraNodeAssignment_4.SetVariables_#6
 ]
}
emitters {
 AuraGlow_Mobile {
  emitterUpdateScript [
   /SetVariables(Emitter.GlowOutlineThickness:NiagaraFloat=&FloatFromCurve(
     FloatCurve=DataInterface() {propertyValues {Curve {keys [<at 0 0 RCIM_Cubic>, <at 0.1 1.5>, {arriveTangent=-1.6667, interpMode=RCIM_Cubic, leaveTangent=-1.6667, time=0.8, value=1.5}, <at 1 0>]}}}
     CurveIndex=$System.NormalizedLoopAge
     Scale Curve=0.2
    )) {enabled=false, isSetParameters=true, module=SetVariables_#7}
   &EmitterState(Life Cycle Mode=::NewEnumerator1, Loop Behavior=::NewEnumerator1, Loop Duration Mode=::NewEnumerator1, UseLoopDelay=false)
   &SpawnBurst_Instantaneous(Spawn Count=11, Use Spawn Probability=false, UseLoopCountLimit=false)
  ]
  particleSpawnScript [
   &InitializeParticle(
    Color=<rgba 1 0.1143 0.0474 1>
    Lifetime=7.33
    Mesh Scale Mode=::NewEnumerator4
    Position Offset=<0 0 100>
    Sprite Size Mode=::NewEnumerator3
    Uniform Sprite Size=100
    UsePositionOffset=true
    Write Material Random=true
   )
   &PartitionParticles(Number Of Partitions=11)
  ]
  particleUpdateScript [
   &ParticleState(Kill Particles When Lifetime Has Elapsed=false, Let Infinitely Lived Particles Die When Emitter Deactivates=true)
   &ScaleSpriteSize(
    Scale Factor=&SelectVector2DFromArray(
     Direct Array Index=$Particles.PartitionParticles.Partition
     Vector 2D Selection Array=DataInterface() {propertyValues {FloatData [<1.25 1.25>, <1.25 1.25>, <2 2>, <1.5 2>, <1.5 1.5>, <1 1>, <1 1>, <0.75 0.75>, <0.75 0.75>, <1.5 2>, <1.5 1.5>]}}
    )
    Scale Sprite Size Mode=::NewEnumerator2
   )
   &ScaleColor(
    Scale Alpha=&FloatFromCurve(
     FloatCurve=DataInterface() {propertyValues {Curve {keys [<at 0 0 RCIM_Cubic>, <at 0.1943 1>, {interpMode=RCIM_Cubic, leaveTangent=-1.2411, time=0.8, value=1}]}}}
     Scale Curve=&MakeCustomFloatFromBool(
      Bool=$User.bIsFrontEndPreview
      True Float=0.18
      False Float=&FN_Select_Float_from_Time_Of_Day(Morning=0.1, Day=0.1, Evening=0.1, Night=0.1)
     )
    )
    ScaleA=true
    ScaleRGB=false
   ) {module=ScaleColor001}
   &SkeletalMeshLocation(
    Bone / Socket Index=$Particles.PartitionParticles.Partition
    Bone Sampling Mode=::NewEnumerator4
    MaskSpawnGroup=false
    Skeletal Mesh=$System.Skeletal Mesh
   )
   &CameraOffset(Camera Offset Amount=&FN_FrontEndFloatModifier(Gameplay=-20, Front End=-50))
   &DynamicMaterialParameters(
    Index 0 Param 1=2
    Index 0 Param 3=0.1
    Index 0 Param 4=1
    Module.Index0Float0Write=true
    Module.Index0Float1Write=true
    Module.Index0Float2Write=true
    Module.Index0Float3Write=true
   )
  ]
  rendererClasses [&NiagaraMeshRendererProperties, &NiagaraSpriteRendererProperties]
  renderers [
   {rendererClass=&NiagaraMeshRendererProperties, rendererIndex=0}
   {rendererClass=&NiagaraSpriteRendererProperties, rendererIndex=1}
  ]
 }
 Flares {
  emitterUpdateScript [
   &EmitterState(
    Inactive Response=::NewEnumerator1
    Life Cycle Mode=::NewEnumerator1
    Loop Behavior=::NewEnumerator1
    Loop Duration Mode=::NewEnumerator1
    UseLoopDelay=false
   )
   &SpawnRate(SpawnRate=21, Use Spawn Probability=false)
  ]
  particleSpawnScript [
   &PartitionParticles(Number Of Partitions=8, Partition Mode:ENiagara_PartitionMode_v2=::NewEnumerator5)
   /SetVariables(Particles.Rands:NiagaraFloat=&RandomRangeFloat(Minimum=0, Maximum=1)) {isSetParameters=true}
   /SetVariables(Particles.PartitionParticles.Partition:NiagaraInt32=&RandomRangeInt(Minimum=0, Maximum=7)) {isSetParameters=true}
   &InitializeParticle(
    Color=<rgba 0.4675 0.5055 1 1>
    Lifetime=&RandomRangeFloat(
     Minimum=&FloatFromCurve(FloatCurve=DataInterface() {propertyValues {Curve {keys [<at 0 0.4>,<at 1 0.9>]}}}, CurveIndex=$Particles.Rands)
     Maximum=0.85
    )
    Sprite Rotation Angle=25
    Sprite Rotation Mode=::NewEnumerator0
    Sprite Size Mode=::NewEnumerator3
    Uniform Sprite Size=&FloatFromCurve(
     FloatCurve=DataInterface() {propertyValues {Curve {keys [<at 0 0.2>,<at 0.85 0.55>,<at 1 1>]}}}
     CurveIndex=$Particles.Rands
     Scale Curve=180
    )
    UsePositionOffset=false
    Write Material Random=true
   )
   &CameraOffset(Camera Offset Amount=&RandomRangeFloat(Minimum=10, Maximum=0))
  ]
  particleUpdateScript [
   &ParticleState()
   &SkeletalMeshLocation(
    Bone / Socket Index=$Particles.PartitionParticles.Partition
    Bone Sampling Mode=::NewEnumerator4
    MaskSpawnGroup=false
    Skeletal Mesh=$System.Skeletal Mesh
   ) {module=SkeletalMeshLocation001}
   &ShapeLocation(
    Additional Quaternion Rotation=false
    MaskSpawnGroup=false
    Non Uniform Scale=&SelectVectorFromArray(
     Direct Array Index=$Particles.PartitionParticles.Partition
     Vector Selection Array=DataInterface() {propertyValues {FloatData [<1 1 0.8>, <2 1 1.5>, <1 1 1>, <1 0.2 0.2>, <0.5 0.2 0.2>, <1 0.2 0.2>, <0.5 0.2 0.2>, <1 1 0.8>, <0.8 0.1 0.1>, <0.8 0.1 0.1>]}}
    )
    Offset=&SelectVectorFromArray(
     Direct Array Index=$Particles.PartitionParticles.Partition
     Vector Selection Array=DataInterface() {propertyValues {FloatData [<0 3 6>, <0 0 0>, <0 0 0>, <5 0 -10>, <0 0 0>, <-5 0 -10>, <0 0 0>, <0 3 6>, <0 10 -20>, <0 10 -20>]}}
    )
    Offset Mode=::NewEnumerator0
    Override Seed=false
    Rotation Mode=::NewEnumerator3
    Rotation Quaternion=$Particles.SkeletalMeshLocation001.SampledRotation
    Shape Origin=$Particles.SkeletalMeshLocation001.SampledPosition
    Sphere Radius=&SelectFloatFromArray(
     Direct Array Index=$Particles.PartitionParticles.Partition
     Float Selection Array=DataInterface() {propertyValues {FloatData [9,8,15,20,20,20,20,9,20,20]}}
    )
    UseOwnerScale=true
   )
   &ScaleSpriteSize(
    Scale Sprite Size Mode=::NewEnumerator0
    Uniform Scale Factor=&Multiply_Float(
     A=&Waveform(
      Global Phase=&Add_Float(A=$Emitter.Age, B=&RandomRangeFloat(Minimum=0, Maximum=1))
      "Waveform [1]"=::NewEnumerator12
      "[1] Frequency"=11
      "[1] Amplitude Min/Max"=&MakeVector2D(X=&Lerp_Float(A=0, B=1, Alpha=$System.AudioReactivity), Y=1)
     )
     B=&FloatFromCurve(FloatCurve=DataInterface() {propertyValues {Curve {
         defaultValue=1
         keys [
          {interpMode=RCIM_Cubic, tangentMode=RCTM_User, time=0, value=0}
          {interpMode=RCIM_Cubic, tangentMode=RCTM_User, time=0.5, value=1}
          {tangentMode=RCTM_User, time=1, value=0}
         ]
        }}})
    )
   )
   &ScaleColor(Scale Alpha=&Lerp_Float(A=0, B=1, Alpha=1), ScaleA=true, ScaleRGB=true)
   /SetVariables(Particles.ChromaFlare:LinearColor=&Multiply_LinearColorByFloat(Linear Color=$Particles.Color, Float=&RandomRangeFloat(Minimum=0.13, Maximum=0.16))) {isSetParameters=true, module=SetVariables_#8}
   /SetVariables(Particles.FlareColor=&MultiplyLinearColor(
     Linear Color=$Particles.Color
     Scale Factor=&MakeVector4FromVectorAndFloat("Vector (RGB)"=<1 1 1>, "Float (Alpha)"=&RandomRangeFloat(Minimum=0.3, Maximum=1))
    )) {isSetParameters=true, module=SetVariables_#9}
  ]
  rendererClasses [&NiagaraSpriteRendererProperties]
  renderers [
   {rendererClass=&NiagaraSpriteRendererProperties}
   {rendererClass=&NiagaraSpriteRendererProperties, rendererIndex=1}
  ]
 }
 FrontEndSun {
  emitterUpdateScript [
   &EmitterState(
    Inactive Response=::NewEnumerator1
    Life Cycle Mode=::NewEnumerator1
    Loop Behavior=::NewEnumerator1
    Loop Duration Mode=::NewEnumerator1
    UseLoopDelay=false
   )
   &SpawnBurst_Instantaneous(Use Spawn Probability=false, UseLoopCountLimit=false)
  ]
  particleSpawnScript [
   &InitializeParticle(
    Color=<rgba 2 2 2 1>
    Position Offset=<0 -40 100>
    Position Offset Coordinate Space=::Local
    Sprite Size Mode=::NewEnumerator3
    Uniform Sprite Size=200
    UsePositionOffset=true
    Write Material Random=true
   )
   &SpriteFacingAndAlignment(Sprite Facing=<0 1 0>, WriteAlignment=false, WriteFacing=true)
  ]
  particleUpdateScript [
   &SolveForcesAndVelocity(Clamp Velocity=false, Limit Acceleration=false)
   &ParticleState(Kill Particles When Lifetime Has Elapsed=false, Loop Particles Lifetime=true)
   &ScaleSpriteSize(Scale Factor=$System.SunSizeGate, Scale Sprite Size Mode=::NewEnumerator2)
   &Light_Attributes(
    Light Radius:NiagaraFloat=&FloatFromCurve(
     FloatCurve=DataInterface() {propertyValues {Curve {keys [<at 0.4 0 RCIM_Cubic>, {arriveTangent=2.0214, interpMode=RCIM_Cubic, leaveTangent=2.0214, tangentMode=RCTM_User, time=1, value=1}]}}}
     CurveIndex=$System.SunProgressSystem
     Scale Curve=&Multiply_Float(A=100, B=$User.SunEnabled)
    )
    Radius=1
   )
  ]
  rendererClasses [&NiagaraSpriteRendererProperties, &NiagaraLightRendererProperties]
  renderers [
   {rendererClass=&NiagaraSpriteRendererProperties}
   {rendererClass=&NiagaraLightRendererProperties, rendererIndex=1}
  ]
 }
 Glow_Body {
  emitterUpdateScript [
   /SetVariables(Emitter.GlowOutlineThickness=1) {isSetParameters=true, module=SetVariables_#7}
   &EmitterState(
    Inactive Response=::NewEnumerator1
    Life Cycle Mode=::NewEnumerator1
    Loop Behavior=::NewEnumerator1
    Loop Duration Mode=::NewEnumerator1
    UseLoopDelay=false
   )
   &SpawnBurst_Instantaneous(Spawn Count=2, Use Spawn Probability=true, UseLoopCountLimit=false)
  ]
  particleSpawnScript [&InitializeParticle(
    Lifetime=10
    Mesh Scale=<1.5 1.5 2.5>
    Mesh Scale Mode=::NewEnumerator4
    Position Offset=&MakeCustomVectorFromBool(Bool=$System.FrontEndActive, True Vector=<0 0 90>, False Vector=<1 0 60>)
    UsePositionOffset=true
    Write Material Random=true
   )]
  particleUpdateScript [
   &ParticleState(Kill Particles When Lifetime Has Elapsed=false, Loop Particles Lifetime=true)
   &ScaleMeshSize(Scale Factor=&Multiply_Vector(
     A=&VectorFromFloat(Value=&FloatFromCurve(FloatCurve=DataInterface() {propertyValues {Curve {keys [<at 0.49 0>,<at 0.5 1>]}}}, CurveIndex=$System.SunProgressSystem))
     B=&MakeCustomVectorFromBool(Bool=$System.FrontEndActive, True Vector=<1 1 0.6>, False Vector=<1 1 1>)
    ))
   &ScaleColor(
    Scale Alpha=&FloatFromCurve(FloatCurve=DataInterface() {propertyValues {Curve {keys [<at 0.49 0>,<at 1 1>]}}}, CurveIndex=$System.SunProgressSystem, Scale Curve=44)
    ScaleA=true
    ScaleRGB=true
   ) {module=ScaleColor001}
   &KillParticles(Kill Particles=&Boolean_AndOperation(
     Bool1=&SetBoolByFloatComparison(A=$User.SunEnabled, B=1, Comparison Type=::NewEnumerator5)
     Bool2=&SetBoolByFloatComparison(A=$Engine.System.Age, B=0.2)
    ))
  ]
  rendererClasses [&NiagaraMeshRendererProperties]
  renderers [
   {rendererClass=&NiagaraMeshRendererProperties, rendererIndex=0}
   {rendererClass=&NiagaraMeshRendererProperties, rendererIndex=1}
  ]
 }
 InstrumentFlares {
  emitterUpdateScript [
   &EmitterState(
    Inactive Response=::NewEnumerator1
    Life Cycle Mode=::NewEnumerator1
    Loop Behavior=::NewEnumerator1
    Loop Duration Mode=::NewEnumerator1
    UseLoopDelay=false
   )
   &SpawnRate(SpawnGroup=1, SpawnRate=20, Use Spawn Probability=false)
   &SpawnRate(SpawnGroup=2, SpawnRate=10, Use Spawn Probability=false) {module=SpawnRate001}
   &SpawnRate(SpawnGroup=3, SpawnRate=12, Use Spawn Probability=false) {module=SpawnRate002}
  ]
  particleSpawnScript [
   /SetVariables(Particles.ActiveInstrument:NiagaraInt32=&MakeIntFromFloat(Float=$User.InstrumentSelect)) {isSetParameters=true}
   /SetVariables(Particles.Rands:NiagaraFloat=&RandomRangeFloat(Minimum=0, Maximum=1)) {isSetParameters=true}
   &PartitionParticles(Number Of Partitions=&SelectIntFromArray(Direct Array Index=$Particles.ActiveInstrument, Int Selection Array=DataInterface() {propertyValues {IntData [0,8,3,3]}}))
   &KillParticles(Kill Particles=&Boolean_NotOperation(Bool=&SetBoolByFloatComparison(
      A=&MakeFloatFromInt(Int=$Emitter.SpawnGroup)
      B=&MakeFloatFromInt(Int=$Particles.ActiveInstrument)
      Comparison Type=::NewEnumerator4
     )))
   &InitializeParticle(
    Color=<rgba 0.4675 0.5055 1 1>
    Lifetime=&RandomRangeFloat(
     Minimum=&FloatFromCurve(FloatCurve=DataInterface() {propertyValues {Curve {keys [<at 0 0.5>,<at 1 0.9>]}}}, CurveIndex=$Particles.Rands)
     Maximum=1.5
    )
    Mesh Scale Mode=::NewEnumerator3
    Mesh Uniform Scale=3
    Sprite Rotation Angle=25
    Sprite Rotation Mode=::NewEnumerator0
    Sprite Size Mode=::NewEnumerator3
    Uniform Sprite Size=&FloatFromCurve(
     FloatCurve=DataInterface() {propertyValues {Curve {keys [<at 0 0.2>,<at 0.85 0.55>,<at 1 1>]}}}
     CurveIndex=$Particles.Rands
     Scale Curve=160
    )
    UsePositionOffset=false
    Write Material Random=true
   )
   &CameraOffset(Camera Offset Amount=&RandomRangeFloat(Minimum=10, Maximum=0))
   /SetVariables(Particles.PartitionParticles.Partition:NiagaraInt32=/Niagara/DynamicInputs/Helpers/MaskInt(
     Mask Source:NiagaraInt32=1
     Mask Parameter:NiagaraInt32=$Particles.ActiveInstrument
     Mask:NiagaraInt32=0
     Masked Int:NiagaraInt32=0
     Passthrough Int:NiagaraInt32=&MaskInt(
      Mask Source:NiagaraInt32=1
      Mask Parameter:NiagaraInt32=$Particles.ActiveInstrument
      Mask:NiagaraInt32=1
      Masked Int:NiagaraInt32=$Particles.PartitionParticles.Partition
      Passthrough Int:NiagaraInt32=&MakeCustomIntFromBool(
       Bool=&SetBoolByFloatComparison(A=&RandomRangeFloat(Minimum=0, Maximum=1), B=0.4)
       True INT=0
       False INT=&MakeCustomIntFromBool(
        Bool=&SetBoolByFloatComparison(A=&RandomRangeFloat(Minimum=0, Maximum=1), B=0.4)
        True INT=1
        False INT=2
       )
      )
     )
    )) {isSetParameters=true}
   /SetVariables(
    Particles.CylinderHeight:NiagaraFloat=&MaskFloat(
     Mask Source=1
     Mask Parameter=&MakeIntFromFloat(Float=$User.InstrumentSelect)
     Mask=0
     Masked Float=0
     Passthrough Float=&MaskFloat(
      Mask Source=1
      Mask Parameter=&MakeIntFromFloat(Float=$User.InstrumentSelect)
      Mask=1
      Masked Float=&SelectFloatFromArray(
       Direct Array Index=$Particles.PartitionParticles.Partition
       Float Selection Array=DataInterface() {propertyValues {FloatData [58,25,35,35,8,2,2,2]}}
      )
      Passthrough Float=&MaskFloat(
       Mask Source=1
       Mask Parameter=&MakeIntFromFloat(Float=$User.InstrumentSelect)
       Mask=2
       Masked Float=&SelectFloatFromArray(Direct Array Index=$Particles.PartitionParticles.Partition, Float Selection Array=DataInterface() {propertyValues {FloatData [10,55,20]}})
       Passthrough Float=&SelectFloatFromArray(Direct Array Index=$Particles.PartitionParticles.Partition, Float Selection Array=DataInterface() {propertyValues {FloatData [10,55,20]}})
      )
     )
    )
    Particles.CylinderOffset:Vector3f=&MaskVector(
     Mask Source=1
     Mask Parameter=&MakeIntFromFloat(Float=$User.InstrumentSelect)
     Mask=0
     Masked Vector=<0 0 0>
     Passthrough Vector=&MaskVector(
      Mask Source=1
      Mask Parameter=&MakeIntFromFloat(Float=$User.InstrumentSelect)
      Mask=1
      Masked Vector=&SelectVectorFromArray(
       Direct Array Index=$Particles.PartitionParticles.Partition
       Vector Selection Array=DataInterface() {propertyValues {FloatData [<-25 84 29>, <15 42 52>, <0 72 80>, <-35 65 80>, <50 40 85>, <30 70 110>, <-53 71 114>, <-80 46 84>]}}
      )
      Passthrough Vector=&MaskVector(
       Mask Source=1
       Mask Parameter=&MakeIntFromFloat(Float=$User.InstrumentSelect)
       Mask=2
       Masked Vector=&SelectVectorFromArray(
        Direct Array Index=$Particles.PartitionParticles.Partition
        Vector Selection Array=DataInterface() {propertyValues {FloatData [<0 0 0>, <0 0 30>, <0 0 65>]}}
       )
       Passthrough Vector=&SelectVectorFromArray(
        Direct Array Index=$Particles.PartitionParticles.Partition
        Vector Selection Array=DataInterface() {propertyValues {FloatData [<0 0 0>, <0 0 30>, <0 0 90>]}}
       )
      )
     )
    )
    Particles.CylinderRadius:NiagaraFloat=&MaskFloat(
     Mask Source=1
     Mask Parameter=&MakeIntFromFloat(Float=$User.InstrumentSelect)
     Mask=0
     Masked Float=0
     Passthrough Float=&MaskFloat(
      Mask Source=1
      Mask Parameter=&MakeIntFromFloat(Float=$User.InstrumentSelect)
      Mask=1
      Masked Float=&SelectFloatFromArray(
       Direct Array Index=$Particles.PartitionParticles.Partition
       Float Selection Array=DataInterface() {propertyValues {FloatData [34,20,20,20,17,20,20,30]}}
      )
      Passthrough Float=&MaskFloat(
       Mask Source=1
       Mask Parameter=&MakeIntFromFloat(Float=$User.InstrumentSelect)
       Mask=2
       Masked Float=&SelectFloatFromArray(Direct Array Index=$Particles.PartitionParticles.Partition, Float Selection Array=DataInterface() {propertyValues {FloatData [20,4,6.8]}})
       Passthrough Float=&SelectFloatFromArray(Direct Array Index=$Particles.PartitionParticles.Partition, Float Selection Array=DataInterface() {propertyValues {FloatData [20,4,8]}})
      )
     )
    )
    Particles.CylinderRotation:Vector3f=&MaskVector(
     Mask Source=1
     Mask Parameter=&MakeIntFromFloat(Float=$User.InstrumentSelect)
     Mask=0
     Masked Vector=<0 0 0>
     Passthrough Vector=&MaskVector(
      Mask Source=1
      Mask Parameter=&MakeIntFromFloat(Float=$User.InstrumentSelect)
      Mask=1
      Masked Vector=&SelectVectorFromArray(
       Direct Array Index=$Particles.PartitionParticles.Partition
       Vector Selection Array=DataInterface() {propertyValues {FloatData [<-0.4 1 0>, <0 0 1>, <0 0.8 1>, <-0.5 0.55 1>, <0 0 1>, <-0.2 -0.3 1>, <0.2 -0.3 1>, <0.2 -0.15 1>]}}
      )
      Passthrough Vector=&MaskVector(
       Mask Source=1
       Mask Parameter=&MakeIntFromFloat(Float=$User.InstrumentSelect)
       Mask=2
       Masked Vector=&SelectVectorFromArray(
        Direct Array Index=$Particles.PartitionParticles.Partition
        Vector Selection Array=DataInterface() {propertyValues {FloatData [<1 0 0>, <0 0 0>, <0 0 0>]}}
       )
       Passthrough Vector=&SelectVectorFromArray(
        Direct Array Index=$Particles.PartitionParticles.Partition
        Vector Selection Array=DataInterface() {propertyValues {FloatData [<1 0 0>, <0 0 0>, <0 0 0>]}}
       )
      )
     )
    )
    Particles.CylinderScale:Vector3f=&MaskVector(
     Mask Source=1
     Mask Parameter=&MakeIntFromFloat(Float=$User.InstrumentSelect)
     Mask=0
     Masked Vector=<1 1 1>
     Passthrough Vector=&MaskVector(
      Mask Source=1
      Mask Parameter=&MakeIntFromFloat(Float=$User.InstrumentSelect)
      Mask=1
      Masked Vector=<1 1 1>
      Passthrough Vector=&MaskVector(
       Mask Source=1
       Mask Parameter=&MakeIntFromFloat(Float=$User.InstrumentSelect)
       Mask=2
       Masked Vector=&SelectVectorFromArray(
        Direct Array Index=$Particles.PartitionParticles.Partition
        Vector Selection Array=DataInterface() {propertyValues {FloatData [<1.5 1 1>, <1 1 1>, <0.3 1 1>]}}
       )
       Passthrough Vector=&SelectVectorFromArray(
        Direct Array Index=$Particles.PartitionParticles.Partition
        Vector Selection Array=DataInterface() {propertyValues {FloatData [<1.5 1 1>, <1 1 1>, <0.3 1 1>]}}
       )
      )
     )
    )
   ) {isSetParameters=true}
   &ShapeLocation(
    Cylinder Height=$Particles.CylinderHeight
    Cylinder Radius=$Particles.CylinderRadius
    MaskSpawnGroup=false
    Non Uniform Scale=$Particles.CylinderScale
    Offset=$Particles.CylinderOffset
    Offset Mode=::NewEnumerator0
    Override Seed=false
    Rotation Axis=$Particles.CylinderRotation
    Rotation Mode=::NewEnumerator1
    Shape Primitive=::NewEnumerator1
    Surface Only=false
    UseOwnerScale=true
   ) {module=ShapeLocation001}
  ]
  particleUpdateScript [
   &ParticleState()
   &SkeletalMeshLocation(
    Bone / Socket Index=10
    Bone Sampling Mode=::NewEnumerator4
    MaskSpawnGroup=false
    Position Sampling=::NewEnumerator2
    Sampled Position Offset:NiagaraPosition=$Particles.ShapeLocation001.ShapePosition
    Skeletal Mesh=$System.Skeletal Mesh
   )
   &ScaleSpriteSize(
    Scale Sprite Size Mode=::NewEnumerator0
    Uniform Scale Factor=&Multiply_Float(
     A=&Waveform(
      Global Phase=&Add_Float(A=$Emitter.Age, B=&RandomRangeFloat(Minimum=0, Maximum=1))
      "Waveform [1]"=::NewEnumerator12
      "[1] Frequency"=11
      "[1] Amplitude Min/Max"=&MakeVector2D(X=&Lerp_Float(A=0.5, B=1, Alpha=$System.AudioReactivity), Y=1)
     )
     B=&FloatFromCurve(FloatCurve=DataInterface() {propertyValues {Curve {
         defaultValue=1
         keys [
          {interpMode=RCIM_Cubic, tangentMode=RCTM_User, time=0, value=0}
          {interpMode=RCIM_Cubic, tangentMode=RCTM_User, time=0.5, value=1}
          {tangentMode=RCTM_User, time=1, value=0}
         ]
        }}})
    )
   )
   &ScaleColor(Scale Alpha=&Lerp_Float(A=0, B=1, Alpha=1), ScaleA=true, ScaleRGB=true)
   /SetVariables(Particles.ChromaFlare:LinearColor=&Multiply_LinearColorByFloat(Linear Color=$Particles.Color, Float=&RandomRangeFloat(Minimum=0.13, Maximum=0.16))) {isSetParameters=true, module=SetVariables_#8}
   /SetVariables(Particles.FlareColor=&MultiplyLinearColor(
     Linear Color=$Particles.Color
     Scale Factor=&MakeVector4FromVectorAndFloat("Vector (RGB)"=<1 1 1>, "Float (Alpha)"=&RandomRangeFloat(Minimum=0.3, Maximum=1))
    )) {isSetParameters=true, module=SetVariables_#9}
   /SetVariables(Particles.Position=/Niagara/DynamicInputs/Helpers/MaskPosition(
     Mask Source:NiagaraInt32=1
     Mask Parameter:NiagaraInt32=$Particles.ActiveInstrument
     Mask:NiagaraInt32=0
     Masked Position:NiagaraPosition=$Particles.Position
     Passthrough Position:NiagaraPosition=&MaskPosition(
      Mask Source:NiagaraInt32=1
      Mask Parameter:NiagaraInt32=$Particles.ActiveInstrument
      Mask:NiagaraInt32=1
      Masked Position:NiagaraPosition=$Particles.Position
      Passthrough Position:NiagaraPosition=$Particles.SkeletalMeshLocation.SampledPosition
     )
    )) {isSetParameters=true}
  ]
  rendererClasses [&NiagaraSpriteRendererProperties]
  renderers [
   {rendererClass=&NiagaraSpriteRendererProperties}
   {rendererClass=&NiagaraSpriteRendererProperties, rendererIndex=1}
  ]
 }
 Minimal {
  emitterUpdateScript [
   &EmitterState(
    Inactive Response=::NewEnumerator1
    Life Cycle Mode=::NewEnumerator1
    Loop Behavior=::NewEnumerator1
    Loop Duration Mode=::NewEnumerator1
    UseLoopDelay=false
   )
   &SpawnBurst_Instantaneous(Use Spawn Probability=false, UseLoopCountLimit=false)
  ]
  particleSpawnScript [
   &InitializeParticle(
    Color=<rgba 2 2 2 1>
    Position Offset=<0 -40 100>
    Position Offset Coordinate Space=::Local
    Sprite Size Mode=::NewEnumerator3
    Uniform Sprite Size=200
    UsePositionOffset=true
    Write Material Random=true
   )
   &SpriteFacingAndAlignment(
    Sprite Facing=/Game/Effects/Niagara/DynamicInputs/Engine/FN_GetSystemAxis(Axis:ENiagaraOrientationAxis=::YAxis, Flip Axis:NiagaraBool=false)
    WriteAlignment=false
    WriteFacing=true
   )
   &ApplyInitialForces()
  ]
  particleUpdateScript [
   &SolveForcesAndVelocity(Clamp Velocity=false, Limit Acceleration=false)
   &ParticleState(Kill Particles When Lifetime Has Elapsed=false, Loop Particles Lifetime=true)
   &ScaleSpriteSize(Scale Factor=$System.SunSizeGate, Scale Sprite Size Mode=::NewEnumerator2)
   &BlendSpace("Position Blend (World <-> Local)":NiagaraFloat=&MakeCustomFloatFromBool(
     Bool=&SetBoolByFloatComparison(A=$System.PlayerSystemDistance, B=50)
     True Float=1
     False Float=0
     Translation Contribution:NiagaraFloat=1
     Rotation Contribution:NiagaraFloat=0
    ))
   &Light_Attributes(
    Light Radius:NiagaraFloat=&FloatFromCurve(
     FloatCurve=DataInterface() {propertyValues {Curve {keys [<at 0.4 0 RCIM_Cubic>, {arriveTangent=2.0214, interpMode=RCIM_Cubic, leaveTangent=2.0214, tangentMode=RCTM_User, time=1, value=1}]}}}
     CurveIndex=$System.SunProgressSystem
     Scale Curve=&Multiply_Float(A=100, B=$User.SunEnabled)
    )
    Radius=1
   )
  ]
  rendererClasses [&NiagaraSpriteRendererProperties, &NiagaraLightRendererProperties]
  renderers [
   {rendererClass=&NiagaraSpriteRendererProperties}
   {rendererClass=&NiagaraLightRendererProperties, rendererIndex=1}
  ]
 }
}
summary {
 systemName=NS_Emote_StrideMiceGiant_Stars
 emitters [
  {emitterName=Flares, rendererClasses [&NiagaraSpriteRendererProperties]}
  {emitterName=Glow_Body, rendererClasses [&NiagaraMeshRendererProperties]}
  {emitterName=Minimal, rendererClasses [&NiagaraSpriteRendererProperties, &NiagaraLightRendererProperties]}
  {emitterName=InstrumentFlares, rendererClasses [&NiagaraSpriteRendererProperties]}
  {emitterName=FrontEndSun, rendererClasses [&NiagaraSpriteRendererProperties, &NiagaraLightRendererProperties]}
  {
   emitterName=AuraGlow_Mobile
   rendererClasses [&NiagaraMeshRendererProperties, &NiagaraSpriteRendererProperties]
  }
 ]
 userVariables [
  User.bIsFrontEnd {type=NiagaraBool}
  User.bIsFrontEndPreview {type=NiagaraBool}
  User.Debug {type=NiagaraFloat, defaultValue=0.1, description=""}
  User.InstrumentSelect {type=NiagaraFloat, defaultValue=0, description=""}
  User.SunEnabled {type=NiagaraFloat, defaultValue=0, description=""}
 ]
}
